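# Corrective Control Rig for the hair bones of the F_MED_ChessBoard female cosmetic's face accessory.
# Five spherical pose readers on head, pelvis and right upper arm write pose weights to variables.
# Only R Arm Up drives an active correction, offsetting right front hair bones through AlphaInterp and ModifyTransforms.
# The third sequence step is gated by a constant-false branch, and two back-hair ModifyTransforms, including the Head Up one, sit outside execution.
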ControlRigBlueprint /BRCosmetics/Characters/Player/Female/Medium/Bodies/F_MED_ChessBoard/Meshes/Parts/F_MED_ChessBoard_FaceAcc_CtrlRig
graphs [
 RigVMFunctionLibrary {links [], nodes []}
 RigVMFunction_Sequence_1_ContainedGraph {nodes [
   n1 = RigVMFunctionEntryNode()
   RigVMFunction_Sequence = &RigVMFunction_Sequence(ExecuteContext=@n1.ExecuteContext)
   n2 = &RigVMFunction_Sequence(ExecuteContext=@RigVMFunction_Sequence.B)
   RigVMFunctionReturnNode(A=@RigVMFunction_Sequence.A, B=@n2.A, C=@n2.B)
  ]}
 RigVMModel {nodes [
   RigUnit_BeginExecution = &RigUnit_BeginExecution()
   RigVMFunction_Sequence = RigVMAggregateNode(ExecuteContext=@RigUnit_BeginExecution.ExecutePin)
   n1 = &RigUnit_SphericalPoseReader(
    ActiveRegionScaleFactors=(PositiveWidth=2,NegativeWidth=0.6,PositiveHeight=2,NegativeHeight=2)
    ActiveRegionSize=0.1
    Debug=(bDrawDebug=true,bDraw2D=False,bDrawLocalAxes=False,DebugScale=25,DebugSegments=20,DebugThickness=0.25)
    DriverAxis=(X=1,Y=0,Z=0)
    DriverItem=(Type=Bone,Name="head")
    FalloffRegionScaleFactors=(PositiveWidth=0.7,NegativeWidth=0.5,PositiveHeight=0.7,NegativeHeight=0.7)
    FalloffSize=0.3
    FlipHeightScaling=False
    FlipWidthScaling=False
    OptionalParentItem=(Type=Bone,Name="None")
    RotationOffset=(X=82.1852,Y=16.0762,Z=-120.4624)
    ExecutePin=@RigVMFunction_Sequence.A
   )
   > RigVMVariableNode(Value=@n1.OutputParam) {variable="Head Up "}
   > n2 = &RigUnit_SphericalPoseReader(
    ActiveRegionScaleFactors=(PositiveWidth=0.3,NegativeWidth=0.7,PositiveHeight=1,NegativeHeight=1)
    ActiveRegionSize=0.1
    Debug=(bDrawDebug=false,bDraw2D=False,bDrawLocalAxes=False,DebugScale=25,DebugSegments=20,DebugThickness=0.25)
    DriverAxis=(X=0,Y=1,Z=0)
    DriverItem=(Type=Bone,Name="pelvis")
    FalloffRegionScaleFactors=(PositiveWidth=0.2,NegativeWidth=0.4,PositiveHeight=0.4,NegativeHeight=0.4)
    FalloffSize=0.22
    FlipHeightScaling=False
    FlipWidthScaling=False
    OptionalParentItem=(Type=Bone,Name="None")
    RotationOffset=(X=70,Y=0,Z=18.2708)
   )
   > RigVMVariableNode(Value=@n2.OutputParam) {variable=Pelvis Low}
   > n3 = &RigUnit_SphericalPoseReader(
    ActiveRegionScaleFactors=(PositiveWidth=1,NegativeWidth=1,PositiveHeight=0.5,NegativeHeight=1)
    ActiveRegionSize=0.1
    Debug=(bDrawDebug=false,bDraw2D=False,bDrawLocalAxes=False,DebugScale=25,DebugSegments=20,DebugThickness=0.25)
    DriverAxis=(X=0,Y=1,Z=0)
    DriverItem=(Type=Bone,Name="head")
    FalloffRegionScaleFactors=(PositiveWidth=1,NegativeWidth=1,PositiveHeight=1,NegativeHeight=1)
    FalloffSize=0.22
    FlipHeightScaling=False
    FlipWidthScaling=False
    OptionalParentItem=(Type=Bone,Name="spine_05")
    RotationOffset=(X=136.2209,Y=72.5786,Z=-19.9062)
   )
   > RigVMVariableNode(Value=@n3.OutputParam) {variable=Look Up Left}
   > n4 = &RigUnit_SphericalPoseReader(
    ActiveRegionScaleFactors=(PositiveWidth=2.2,NegativeWidth=2.2,PositiveHeight=2.2,NegativeHeight=2.2)
    ActiveRegionSize=0.11
    Debug=(bDrawDebug=true,bDraw2D=false,bDrawLocalAxes=true,DebugScale=25,DebugSegments=20,DebugThickness=0.25)
    DriverAxis=(X=-1,Y=0,Z=0)
    DriverItem=(Type=Bone,Name="upperarm_r")
    FalloffRegionScaleFactors=(PositiveWidth=1,NegativeWidth=1,PositiveHeight=4,NegativeHeight=1)
    FalloffSize=0.3
    FlipHeightScaling=false
    FlipWidthScaling=true
    OptionalParentItem=(Type=Bone,Name="spine_02")
    RotationOffset=(X=-140,Y=50,Z=0)
   )
   > RigVMVariableNode(Value=@n4.OutputParam) {variable=R Arm Up}
   > n5 = &RigUnit_SphericalPoseReader(
    ActiveRegionScaleFactors=(PositiveWidth=1,NegativeWidth=1,PositiveHeight=1,NegativeHeight=1)
    ActiveRegionSize=0.1
    Debug=(bDrawDebug=false,bDraw2D=False,bDrawLocalAxes=False,DebugScale=25,DebugSegments=20,DebugThickness=0.25)
    DriverAxis=(X=1,Y=0,Z=0)
    DriverItem=(Type=Bone,Name="head")
    FalloffRegionScaleFactors=(PositiveWidth=0.7,NegativeWidth=0.7,PositiveHeight=0.7,NegativeHeight=0.7)
    FalloffSize=0.22
    FlipHeightScaling=False
    FlipWidthScaling=False
    OptionalParentItem=(Type=Bone,Name="None")
    RotationOffset=(X=94.7838,Y=-26.3245,Z=-142.8814)
   )
   > RigVMVariableNode(Value=@n5.OutputParam) {variable=Head Back Drastic}
   n6 = &RigVMFunction_AlphaInterp(
    ClampMax=0.8
    InterpSpeedDecreasing=10
    InterpSpeedIncreasing=10
    Scale=1
    bClampResult=False
    bInterpResult=False
    Value=RigVMVariableNode() {variable=R Arm Up}.Value
   )
   &RigUnit_ModifyTransforms(
    ItemToModify="((Item=(Type=Bone,Name=\"base_fr_hair_r\"),Transform=(Rotation=(X=0,Y=0,Z=0,W=1),Translation=(X=0,Y=0,Z=0),Scale3D=(X=1,Y=1,Z=1))),(Item=(Type=Bone,Name=\"dyn_fr_lowerhair_r_01\"),Transform=(Rotation=(X=0.0151,Y=0.173,Z=0.0858,W=0.9811),Translation=(X=-0.5656, ...(truncated)"
    WeightMinimum=0
    Weight=@n6.Result
    ExecutePin=@RigVMFunction_Sequence.B
   )
   n7 = &RigVMFunction_ControlFlowBranch(Condition=False, ExecuteContext=@RigVMFunction_Sequence.C)
   &RigUnit_ModifyTransforms(
    ItemToModify="((Item=(Type=Bone,Name=\"None\"),Transform=(Rotation=(X=0,Y=0,Z=0,W=1),Translation=(X=0,Y=0,Z=1.6874),Scale3D=(X=1,Y=1,Z=1))),(Item=(Type=Bone,Name=\"None\"),Transform=(Rotation=(X=0,Y=-0.1736,Z=0,W=0.9848),Translation=(X=-0.2566,Y=0,Z=2),Scale3D=(X=1,Y=1,Z=1) ...(truncated)"
    Weight=1
    WeightMinimum=0
    ExecutePin=@n7.True
   )
   AlphaInterp = &RigVMFunction_AlphaInterp(
    ClampMax=0.8
    InterpSpeedDecreasing=10
    InterpSpeedIncreasing=10
    Scale=1
    bClampResult=False
    bInterpResult=False
    Value=RigVMVariableNode() {variable="Head Up "}.Value
   )
   &RigUnit_ModifyTransforms(
    ExecutePin=()
    ItemToModify="((Item=(Type=Bone,Name=\"base_backhair_l\"),Transform=(Rotation=(X=0,Y=0,Z=0,W=1),Translation=(X=-0.565,Y=0,Z=0.3),Scale3D=(X=1,Y=1,Z=1))),(Item=(Type=Bone,Name=\"base_backhair_m\"),Transform=(Rotation=(X=0,Y=-0.1305,Z=0,W=0.9914),Translation=(X=-0.7449,Y=0,Z= ...(truncated)"
    WeightMinimum=0
    Weight=@AlphaInterp.Result
   )
   &RigUnit_ModifyTransforms(
    ExecutePin=()
    ItemToModify="((Item=(Type=Bone,Name=\"base_backhair_l\"),Transform=(Rotation=(X=0,Y=0,Z=0,W=1),Translation=(X=0,Y=0,Z=1.27),Scale3D=(X=1,Y=1,Z=1.5))),(Item=(Type=Bone,Name=\"base_backhair_m\"),Transform=(Rotation=(X=0,Y=0,Z=0,W=1),Translation=(X=0,Y=0.5089,Z=-0.5209),Scale ...(truncated)"
    Weight=1
    WeightMinimum=0
   )
  ]}
]
hierarchy [
 root {loc=<0 0 0>, rot=<0 0 0 1>}
 pelvis {parent=root, loc=<0 -0.5301 98.0571>, rot=<0.0127 0.707 -0.0127 -0.707>}
 spine_01 {parent=pelvis, loc=<6.1208 -0.9328 0>, rot=<0 0 0.034 -0.9994>}
 spine_02 {parent=spine_01, loc=<6.5996 -0.2095 0>, rot=<0 0 0.0238 -0.9997>}
 spine_03 {parent=spine_02, loc=<6.4914 -0.1237 0>, rot=<0 0 0.0149 -0.9999>}
 spine_04 {parent=spine_03, loc=<6.9598 0.0794 0>, rot=<0 0 -0.0132 -0.9999>}
 spine_05 {parent=spine_04, loc=<6.9987 0 0>, rot=<0 0 -0.0015 -1>}
 neck_01 {parent=spine_05, loc=<11.3138 0.1529 0>, rot=<0 0 -0.2003 -0.9797>}
 neck_02 {parent=neck_01, loc=<4.5627 0 0>, rot=<0 0 0.0504 -0.9987>}
 head {parent=neck_02, loc=<5.3914 0 0>, rot=<0 0 0.115 -0.9934>}
 base_hair {parent=head, loc=<0 0 0>, rot=<0.7071 -0.0009 0.0009 -0.7071>}
 base_hairtop {parent=base_hair, loc=<12.0279 0 0.1313>, rot=<0 0 0 1>}
 base_fr_hair_l {parent=base_hair, loc=<4.5407 6.7499 -2.4319>, rot=<-0.0574 -0.7279 -0.6832 -0.0073>}
 dyn_fr_hair_l {parent=base_fr_hair_l, loc=<0 0 0>, rot=<0 0 0 1>}
 dyn_fr_upperhair_l_01 {parent=dyn_fr_hair_l, loc=<2.4773 1.7752 0.5223>, rot=<0.2202 -0.0132 -0.1899 -0.9567>}
 dyn_fr_upperhair_l_02 {parent=dyn_fr_upperhair_l_01, loc=<5.82 -0.0237 0.0351>, rot=<0.0087 -0.1065 -0.0808 -0.991>}
 dyn_fr_lowerhair_l_01 {parent=dyn_fr_hair_l, loc=<8.4313 -0.7139 -0.4399>, rot=<-0.0083 0.1065 -0.1804 -0.9778>}
 dyn_fr_lowerhair_l_02 {parent=dyn_fr_lowerhair_l_01, loc=<5.104 0.097 0.05>, rot=<0.0841 0.0911 -0.23 -0.9653>}
 base_fr_hair_r {parent=base_hair, loc=<2.2688 -7.7699 -1.1382>, rot=<0.0741 -0.7209 0.6754 0.1364>}
 dyn_fr_hair_r {parent=base_fr_hair_r, loc=<0 0 0>, rot=<0 0 0 1>}
 dyn_fr_midhair_r {parent=dyn_fr_hair_r, loc=<6.6756 -0.6172 0.0848>, rot=<-0.1033 0.0283 0.1547 -0.9821>}
 dyn_fr_upperhair_r {parent=dyn_fr_hair_r, loc=<3.9886 -3.6191 -0.1055>, rot=<-0.0579 -0.1292 0.072 -0.9873>}
 base_backhair_l {parent=base_hair, loc=<3.9306 4.743 -6.5659>, rot=<0 -1 0 -0.0055>}
 dyn_backhair_l_01 {parent=base_backhair_l, loc=<0 0 0>, rot=<0 0 0 1>}
 dyn_backhair_l_02 {parent=dyn_backhair_l_01, loc=<4.9315 0 0>, rot=<0 0.0741 0 -0.9973>}
 dyn_backhair_l_03 {parent=dyn_backhair_l_02, loc=<4.533 0 0>, rot=<0 0.0437 0 -0.999>}
 dyn_backhair_l_04 {parent=dyn_backhair_l_03, loc=<4.3331 0 0>, rot=<0 0.0215 0 -0.9998>}
 dyn_backhair_l_05 {parent=dyn_backhair_l_03, loc=<2.8297 3.7301 -0.1834>, rot=<0.3879 -0.0343 -0.4632 -0.7961>}
 base_backhair_m {parent=base_hair, loc=<3.9397 0 -7.4037>, rot=<0 -0.9998 0 -0.021>}
 dyn_backhair_m_01 {parent=base_backhair_m, loc=<0 0 0>, rot=<0 0 0 1>}
 dyn_backhair_m_02 {parent=dyn_backhair_m_01, loc=<4.9315 0 0>, rot=<0 0.0741 0 -0.9973>}
 dyn_backhair_m_03 {parent=dyn_backhair_m_02, loc=<4.533 0 0>, rot=<0 0.0437 0 -0.999>}
 dyn_backhair_m_04 {parent=dyn_backhair_m_03, loc=<4.3331 0 0>, rot=<0 0.0215 0 -0.9998>}
 dyn_backhair_m_05 {parent=dyn_backhair_m_04, loc=<4.3842 0 0>, rot=<0 0.0775 0 -0.997>}
 base_backhair_r {parent=base_hair, loc=<3.9306 -4.7428 -6.5659>, rot=<0 -0.9998 0 -0.021>}
 dyn_backhair_r_01 {parent=base_backhair_r, loc=<0 0 0>, rot=<0 0 0 1>}
 dyn_backhair_r_02 {parent=dyn_backhair_r_01, loc=<4.9315 0 0>, rot=<0 0.0741 0 -0.9973>}
 dyn_backhair_r_03 {parent=dyn_backhair_r_02, loc=<4.533 0 0>, rot=<0 0.0437 0 -0.999>}
 dyn_backhair_r_04 {parent=dyn_backhair_r_03, loc=<4.3331 0 0>, rot=<0 0.0215 0 -0.9998>}
 dyn_backhair_r_05 {parent=dyn_backhair_r_03, loc=<3.6075 -4.4111 -1.5382>, rot=<-0.4054 -0.0108 0.6665 -0.6255>}
 base_fr_02_hair_r {parent=base_hair, loc=<3.6288 -7.912 -3.4577>, rot=<-0.0166 -0.7702 0.6374 0.0137>}
 dyn_fr_02_hair_r {parent=base_fr_02_hair_r, loc=<0 0 0>, rot=<0 0 0 1>}
 dyn_fr_lowerhair_r_01 {parent=dyn_fr_02_hair_r, loc=<8.588 -0.1165 -0.0261>, rot=<-0.2117 0.2124 0.3584 -0.8841>}
 dyn_fr_lowerhair_r_02 {parent=dyn_fr_lowerhair_r_01, loc=<5.3673 0.9426 0.0015>, rot=<-0.2107 -0.1666 0.1755 -0.9471>}
 base_frBang_hair_A {parent=base_hair, loc=<8.9732 -4.0643 9.1736>, rot=<0.0107 -0.3264 0.9419 -0.0783>}
 dyn_frBang_hair_A_01 {parent=base_frBang_hair_A, loc=<0 0 0>, rot=<0 0 0 1>}
 dyn_frBang_hair_A_02 {parent=dyn_frBang_hair_A_01, loc=<3.139 0 0>, rot=<0.0233 -0.0783 -0.1593 -0.9839>}
 base_frBang_hair_E {parent=base_hair, loc=<5.3397 7.8556 3.9829>, rot=<-0.2537 0.5023 0.6864 0.4605>}
 dyn_frBang_hair_E_01 {parent=base_frBang_hair_E, loc=<0 0 0>, rot=<0 0 0 1>}
 base_frBang_hair_D {parent=base_hair, loc=<9.4126 3.0466 9.1354>, rot=<-0.0113 0.2594 0.902 0.3448>}
 dyn_frBang_hair_D_01 {parent=base_frBang_hair_D, loc=<0 0 0>, rot=<0 0 0 1>}
 dyn_frBang_hair_D_02 {parent=dyn_frBang_hair_D_01, loc=<4.3813 0 0>, rot=<-0.1897 -0.2905 0.1735 -0.9217>}
 base_frBang_hair_C {parent=base_hair, loc=<9.6435 -0.2837 9.7634>, rot=<0.0483 0.0806 0.9574 0.2731>}
 dyn_frBang_hair_C_01 {parent=base_frBang_hair_C, loc=<0 0 0>, rot=<0 0 0 1>}
 base_frBang_hair_B {parent=base_hair, loc=<9.5127 -1.7996 9.781>, rot=<0.0809 -0.0482 0.9904 -0.1011>}
 dyn_frBang_hair_B_01 {parent=base_frBang_hair_B, loc=<0 0 0>, rot=<0 0 0 1>}
 clavicle_l {parent=spine_05, loc=<5.4831 -0.1053 -3.2946>, rot=<0.0465 -0.7289 -0.009 -0.6829>}
 upperarm_l {parent=clavicle_l, loc=<9.8658 -1.2136 -0.4105>, rot=<0.0164 -0.3245 -0.0269 -0.9454>}
 clavicle_r {parent=spine_05, loc=<5.4831 -0.1053 3.2946>, rot=<0.7289 0.0465 -0.6829 0.009>}
 upperarm_r {parent=clavicle_r, loc=<-9.8658 1.2136 0.4105>, rot=<0.0164 -0.3245 -0.0269 -0.9454>}
]
variables [
 Head Back Drastic {type=double, default=0}
 "Head Up " {type=double, default=0}
 Look Up Left {type=double, default=0}
 Pelvis Low {type=double, default=0}
 R Arm Up {type=double, default=0}
 is crouching {type=bool, default=False}
]
