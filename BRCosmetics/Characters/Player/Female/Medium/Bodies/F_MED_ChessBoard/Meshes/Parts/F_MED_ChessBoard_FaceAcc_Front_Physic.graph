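# Physics asset for the front face accessory mesh of the female ChessBoard cosmetic body, simulating its hair.
# Capsule chains cover three back hair strands and branching front locks, with head, neck, upper spine, clavicle and upper arm capsules as body volumes.
# Most joints limit swing with little or no twist, while one front-right base is angularly locked to the head.
# Extra angularly free constraints cross-link neighboring strands.
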PhysicsAsset /BRCosmetics/Characters/Player/Female/Medium/Bodies/F_MED_ChessBoard/Meshes/Parts/F_MED_ChessBoard_FaceAcc_Front_Physic
bodies {
 base_backhair_l [{shapeType=Capsule, length=0.51, radius=0.5, shapeName=base_backhair_l_capsule, center=<1.5009 0 0>, rotation=<rot 90 0 0>}]
 base_backhair_m [{shapeType=Capsule, length=0.51, radius=0.5, shapeName=base_backhair_m_capsule, center=<1.5009 0 0>, rotation=<rot 90 0 0>}]
 base_backhair_r [{shapeType=Capsule, length=0.505, radius=0.505, shapeName=base_backhair_r_capsule, center=<1.5009 0 0>, rotation=<rot 90 0 0>}]
 base_fr_02_hair_r [{shapeType=Capsule, length=1, radius=1, shapeName=base_fr_02_hair_r_capsule}]
 base_fr_hair_l [{shapeType=Capsule, length=0.505, radius=0.505, shapeName=base_fr_hair_l_capsule, rotation=<rot 0 -90 0>}]
 base_fr_hair_r [{shapeType=Capsule, length=0.505, radius=0.505, shapeName=base_fr_hair_r_capsule, rotation=<rot 0 -90 0>}]
 clavicle_l [{
   shapeType=Capsule
   length=8
   radius=6.75
   shapeName=clavicle_l_capsule
   center=<6.7504 -0.9353 -2.5369>
   rotation=<rot 88.6214 -0.0148 179.9852>
  }]
 clavicle_r [{
   shapeType=Capsule
   length=9
   radius=6.25
   shapeName=clavicle_l_capsule
   center=<-5.5045 0.6582 2.0657>
   rotation=<rot 90 -179.8417 -179.8418>
  }]
 dyn_backhair_l_01 [{shapeType=Capsule, length=2.0257, radius=0.7781, shapeName=dyn_backhair_l_01_capsule, center=<1.5009 0 0.5473>, rotation=<rot 90 0 0>}]
 dyn_backhair_l_02 [{shapeType=Capsule, length=1.5, radius=1.5, shapeName=dyn_backhair_l_02_capsule, center=<1.5009 0 0.5473>, rotation=<rot 90 0 0>}]
 dyn_backhair_l_03 [{shapeType=Capsule, length=1.5, radius=2, shapeName=dyn_backhair_l_03_capsule, center=<1.5009 0 0>, rotation=<rot 90 0 0>}]
 dyn_backhair_l_04 [{shapeType=Capsule, length=2, radius=2.3, shapeName=dyn_backhair_l_04_capsule, center=<2.8579 1.3297 1.7461>, rotation=<rot 90 0 -10>}]
 dyn_backhair_l_05 [{
   shapeType=Capsule
   length=2.5
   radius=2
   shapeName=dyn_backhair_l_05_capsule
   center=<2.7733 -0.4631 -0.3087>
   rotation=<rot 60.4131 173.7054 173.0917>
  }]
 dyn_backhair_m_01 [{shapeType=Capsule, length=2.5283, radius=1.4674, shapeName=dyn_backhair_m_01_capsule, center=<1.5009 0 0.5473>, rotation=<rot 90 0 0>}]
 dyn_backhair_m_02 [{shapeType=Capsule, length=2, radius=1.5, shapeName=dyn_backhair_m_02_capsule, center=<1.5009 0 0.5473>, rotation=<rot 90 0 0>}]
 dyn_backhair_m_03 [{shapeType=Capsule, length=1.5, radius=1.5, shapeName=dyn_backhair_m_03_capsule, center=<1.5009 0 0>, rotation=<rot 90 0 0>}]
 dyn_backhair_m_04 [{shapeType=Capsule, length=1.5, radius=2, shapeName=dyn_backhair_m_04_capsule, center=<1.5009 0 1.0897>, rotation=<rot 90 0 0>}]
 dyn_backhair_m_05 [{shapeType=Capsule, length=1, radius=3, shapeName=dyn_backhair_m_05_capsule, center=<1.421 0 -0.0149>, rotation=<rot 80 0 0>}]
 dyn_backhair_r_01 [{shapeType=Capsule, length=0.505, radius=0.505, shapeName=dyn_backhair_r_01_capsule, center=<1.5009 0 0.5473>, rotation=<rot 90 0 0>}]
 dyn_backhair_r_02 [{shapeType=Capsule, length=2, radius=1.5, shapeName=dyn_backhair_r_02_capsule, center=<1.5009 0 0.5473>, rotation=<rot 90 0 0>}]
 dyn_backhair_r_03 [{shapeType=Capsule, length=2, radius=1.5, shapeName=dyn_backhair_r_03_capsule, center=<1.5009 0 0>, rotation=<rot 90 0 0>}]
 dyn_backhair_r_04 [{shapeType=Capsule, length=3, radius=2.25, shapeName=dyn_backhair_r_04_capsule, center=<4.5686 0 1.1192>, rotation=<rot 90 0 0>}]
 dyn_backhair_r_05 [{
   shapeType=Capsule
   length=4
   radius=2
   shapeName=dyn_backhair_r_05_capsule
   center=<0.657 1.301 -0.1617>
   rotation=<rot 70.0001 -140.001 -90.0007>
  }]
 dyn_fr_02_hair_r [{shapeType=Capsule, length=4, radius=1.5, shapeName=dyn_fr_02_hair_r_capsule, center=<5.1605 0 0>, rotation=<rot 90 0 0>}]
 dyn_fr_hair_l [{shapeType=Capsule, length=3.7578, radius=2, shapeName=dyn_fr_hair_l_capsule, center=<4.851 -0.6604 0>, rotation=<rot 90 -179.4032 -174.4183>}]
 dyn_fr_hair_r [{shapeType=Capsule, length=6, radius=2, shapeName=dyn_fr_hair_r_capsule, center=<3.9831 0 0>, rotation=<rot -90 179.156 -179.1564>}]
 dyn_fr_lowerhair_l_01 [
  {
   shapeType=Capsule
   length=2
   radius=2
   shapeName=dyn_fr_lowerhair_l_01_capsule
   center=<5.5489 -2.1299 1.2886>
   rotation=<rot 71.3025 97.7037 -86.9263>
  }
  {
   shapeType=Capsule
   length=5
   radius=1.5
   shapeName=dyn_fr_lowerhair_l_01_capsule
   center=<4.8399 1.3793 0.1564>
   rotation=<rot 83.1489 6.0956 -21.3685>
  }
 ]
 dyn_fr_lowerhair_l_02 [{
   shapeType=Capsule
   length=5
   radius=3
   shapeName=dyn_fr_lowerhair_l_02_capsule
   center=<3.6642 -0.1863 -1.6928>
   rotation=<rot 21.4832 -60.7801 -48.6098>
  }]
 dyn_fr_lowerhair_r_01 [{shapeType=Capsule, length=3, radius=3, shapeName=dyn_fr_lowerhair_l_01_capsule, center=<1.7671 0.4027 0>, rotation=<rot 90 0.0015 0.0019>}]
 dyn_fr_lowerhair_r_02 [{
   shapeType=Capsule
   length=5.5607
   radius=2.3389
   shapeName=dyn_fr_lowerhair_r_02_capsule
   center=<3.6537 -0.5891 -0.4924>
   rotation=<rot 0 -95.132 -72.0506>
  }]
 dyn_fr_midhair_r [{shapeType=Capsule, length=4, radius=1.25, shapeName=dyn_fr_upperhair_l_02_capsule, center=<3.1432 0 -0.4249>, rotation=<rot 65 -0.0001 0>}]
 dyn_fr_upperhair_l_01 [
  {
   shapeType=Capsule
   length=4.5
   radius=0.9
   shapeName=dyn_fr_upperhair_l_01_capsule
   center=<4.17 2.4549 0.708>
   rotation=<rot 83.0394 46.4704 26.1272>
  }
  {shapeType=Capsule, length=2.5, radius=1.5, shapeName=dyn_fr_upperhair_l_01_capsule, center=<3.4616 -0.0012 0>, rotation=<rot 90 0 0>}
 ]
 dyn_fr_upperhair_l_02 [{shapeType=Capsule, length=4, radius=1.75, shapeName=dyn_fr_upperhair_l_02_capsule, center=<2.4678 0 -0.2867>, rotation=<rot 90 0 0>}]
 dyn_fr_upperhair_r [{shapeType=Capsule, length=4, radius=1.75, shapeName=dyn_fr_upperhair_l_02_capsule, center=<4.1623 0 0>, rotation=<rot 90 0 0>}]
 head [
  {shapeType=Capsule, length=11.5, radius=5.6, shapeName=neck_01_capsule, center=<-0.9751 0.2907 0>, rotation=<rot 0 23.5791 90>}
  {shapeType=Capsule, length=5.5, radius=7, shapeName=head_capsule, center=<4.9029 -0.2275 0>, rotation=<rot 0 0 90>}
 ]
 neck_01 [{shapeType=Capsule, length=5.005, radius=6.005, shapeName=neck_01_capsule, center=<0.7758 1.2827 0>, rotation=<rot 0 76.2098 90>}]
 neck_02 [{shapeType=Capsule, length=1.005, radius=5.95, shapeName=neck_01_capsule, center=<-0.6036 -1.023 0.8056>, rotation=<rot 0 87.0268 90>}]
 spine_05 [{shapeType=Capsule, length=26.505, radius=7.505, shapeName=spine_05_capsule, center=<0.9635 -1.1514 0>, rotation=<rot 0 -90 0>}]
 upperarm_l [{
   shapeType=Capsule
   length=19.891
   radius=8.1
   shapeName=upperarm_r_capsule
   center=<14.4822 -0.22 -0.9631>
   rotation=<rot -87.6171 -2.4958 -171.0809>
  }]
 upperarm_r [{
   shapeType=Capsule
   length=21.391
   radius=6.6
   shapeName=upperarm_r_capsule
   center=<-12.3679 0.7215 0.5538>
   rotation=<rot -87.3286 -165.6468 167.054>
  }]
}
constraints [
 {bone1Name=dyn_fr_upperhair_r, bone2Name=dyn_fr_hair_r, swing1LimitDegrees=20, swing2LimitDegrees=20, twistLimitDegrees=0}
 {
  bone1Name=upperarm_r
  bone2Name=clavicle_r
  swing1LimitDegrees=45
  swing1Motion=Free
  swing2LimitDegrees=45
  swing2Motion=Free
  twistLimitDegrees=45
  twistMotion=Free
 }
 {bone1Name=base_backhair_l, bone2Name=head, swing1LimitDegrees=20, swing2LimitDegrees=20, twistLimitDegrees=0}
 {bone1Name=dyn_backhair_l_01, bone2Name=base_backhair_l, swing1LimitDegrees=20, swing2LimitDegrees=35, twistLimitDegrees=0}
 {bone1Name=dyn_backhair_l_02, bone2Name=dyn_backhair_l_01, swing1LimitDegrees=20, swing2LimitDegrees=20, twistLimitDegrees=0}
 {bone1Name=dyn_backhair_l_03, bone2Name=dyn_backhair_l_02, swing1LimitDegrees=15, swing2LimitDegrees=35, twistLimitDegrees=0}
 {bone1Name=dyn_backhair_l_04, bone2Name=dyn_backhair_l_03, swing1LimitDegrees=20, swing2LimitDegrees=40, twistLimitDegrees=0}
 {bone1Name=dyn_backhair_l_05, bone2Name=dyn_backhair_l_03, swing1LimitDegrees=30, swing2LimitDegrees=30, twistLimitDegrees=5}
 {bone1Name=base_backhair_m, bone2Name=head, swing1LimitDegrees=20, swing2LimitDegrees=20, twistLimitDegrees=0}
 {bone1Name=dyn_backhair_m_01, bone2Name=base_backhair_m, swing1LimitDegrees=20, swing2LimitDegrees=35, twistLimitDegrees=0}
 {bone1Name=dyn_backhair_m_02, bone2Name=dyn_backhair_m_01, swing1LimitDegrees=20, swing2LimitDegrees=20, twistLimitDegrees=0}
 {bone1Name=dyn_backhair_m_03, bone2Name=dyn_backhair_m_02, swing1LimitDegrees=20, swing2LimitDegrees=50, twistLimitDegrees=0}
 {bone1Name=dyn_backhair_m_04, bone2Name=dyn_backhair_m_03, swing1LimitDegrees=20, swing2LimitDegrees=40, twistLimitDegrees=0}
 {bone1Name=dyn_backhair_m_05, bone2Name=dyn_backhair_m_04, swing1LimitDegrees=20, swing2LimitDegrees=40, twistLimitDegrees=0}
 {bone1Name=base_backhair_r, bone2Name=head, swing1LimitDegrees=20, swing2LimitDegrees=20, twistLimitDegrees=0}
 {bone1Name=dyn_backhair_r_01, bone2Name=base_backhair_r, swing1LimitDegrees=20, swing2LimitDegrees=35, twistLimitDegrees=0}
 {bone1Name=dyn_backhair_r_02, bone2Name=dyn_backhair_r_01, swing1LimitDegrees=20, swing2LimitDegrees=20, twistLimitDegrees=0}
 {bone1Name=dyn_backhair_r_03, bone2Name=dyn_backhair_r_02, swing1LimitDegrees=15, swing2LimitDegrees=35, twistLimitDegrees=0}
 {bone1Name=dyn_backhair_r_04, bone2Name=dyn_backhair_r_03, swing1LimitDegrees=20, swing2LimitDegrees=40, twistLimitDegrees=0}
 {bone1Name=dyn_backhair_r_05, bone2Name=dyn_backhair_r_03, swing1LimitDegrees=20, swing2LimitDegrees=30, twistLimitDegrees=0}
 {
  bone1Name=base_fr_02_hair_r
  bone2Name=head
  swing1LimitDegrees=45
  swing1Motion=Locked
  swing2LimitDegrees=45
  swing2Motion=Locked
  twistLimitDegrees=45
  twistMotion=Locked
 }
 {bone1Name=dyn_fr_02_hair_r, bone2Name=base_fr_02_hair_r, swing1LimitDegrees=20, swing2LimitDegrees=20, twistLimitDegrees=0}
 {bone1Name=dyn_fr_lowerhair_r_01, bone2Name=dyn_fr_02_hair_r, swing1LimitDegrees=20, swing2LimitDegrees=15, twistLimitDegrees=5}
 {bone1Name=dyn_fr_lowerhair_r_02, bone2Name=dyn_fr_lowerhair_r_01, swing1LimitDegrees=15, swing2LimitDegrees=15, twistLimitDegrees=1}
 {bone1Name=dyn_fr_hair_l, bone2Name=base_fr_hair_l, swing1LimitDegrees=20, swing2LimitDegrees=20, twistLimitDegrees=0}
 {bone1Name=dyn_fr_lowerhair_l_01, bone2Name=dyn_fr_hair_l, swing1LimitDegrees=20, swing2LimitDegrees=20, twistLimitDegrees=10}
 {bone1Name=dyn_fr_lowerhair_l_02, bone2Name=dyn_fr_lowerhair_l_01, swing1LimitDegrees=25, swing2LimitDegrees=20, twistLimitDegrees=0}
 {bone1Name=dyn_fr_upperhair_l_01, bone2Name=dyn_fr_hair_l, swing1LimitDegrees=5, swing2LimitDegrees=10, twistLimitDegrees=0}
 {bone1Name=dyn_fr_upperhair_l_02, bone2Name=dyn_fr_upperhair_l_01, swing1LimitDegrees=5, swing2LimitDegrees=10, twistLimitDegrees=0}
 {bone1Name=dyn_fr_hair_r, bone2Name=base_fr_hair_r, swing1LimitDegrees=20, swing2LimitDegrees=20, twistLimitDegrees=5}
 {bone1Name=dyn_fr_midhair_r, bone2Name=dyn_fr_hair_r, swing1LimitDegrees=20, swing2LimitDegrees=20, twistLimitDegrees=0}
 {
  bone1Name=dyn_backhair_m_05
  bone2Name=dyn_backhair_r_04
  swing1LimitDegrees=45
  swing1Motion=Free
  swing2LimitDegrees=45
  swing2Motion=Free
  twistLimitDegrees=45
  twistMotion=Free
 }
 {
  bone1Name=dyn_backhair_m_05
  bone2Name=dyn_backhair_l_04
  swing1LimitDegrees=45
  swing1Motion=Free
  swing2LimitDegrees=45
  swing2Motion=Free
  twistLimitDegrees=45
  twistMotion=Free
 }
 {
  bone1Name=dyn_backhair_m_02
  bone2Name=dyn_backhair_l_02
  swing1LimitDegrees=45
  swing1Motion=Free
  swing2LimitDegrees=45
  swing2Motion=Free
  twistLimitDegrees=45
  twistMotion=Free
 }
 {
  bone1Name=dyn_backhair_m_02
  bone2Name=dyn_backhair_r_02
  swing1LimitDegrees=45
  swing1Motion=Free
  swing2LimitDegrees=45
  swing2Motion=Free
  twistLimitDegrees=45
  twistMotion=Free
 }
 {
  bone1Name=dyn_backhair_r_05
  bone2Name=dyn_backhair_r_04
  swing1LimitDegrees=45
  swing1Motion=Free
  swing2LimitDegrees=45
  swing2Motion=Free
  twistLimitDegrees=45
  twistMotion=Free
 }
 {
  bone1Name=dyn_fr_lowerhair_l_01
  bone2Name=dyn_backhair_l_03
  swing1LimitDegrees=45
  swing1Motion=Free
  swing2LimitDegrees=45
  swing2Motion=Free
  twistLimitDegrees=45
  twistMotion=Free
 }
]
