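# Post-physics AnimBP for the ChessBoard female face accessory, driving its front, bang and back hair.
# Copied mesh poses feed a Control Rig and head-up tilt, then one of two RigidBody simulations chosen by front-end state.
# Constraints, arm-raise corrections, AnimDynamics and fixed bone offsets then refine the hair.
# The EventGraph computes gravity override and head/arm alphas; its backbling values go unread by the AnimGraph.
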AnimBlueprint /BRCosmetics/Characters/Player/Female/Medium/Bodies/F_MED_ChessBoard/Meshes/Parts/F_MED_ChessBoard_FaceAcc_AnimBP
target_skeleton=&/F_MED_ChessBoard_FaceAcc_Skeleton
graphs [
 AnimGraph {nodes [
   n1 = RigidBody!(
    ComponentPose=Local To Component(LocalPose=Use cached pose 'Pre_RBAN_2_RBAN'())
    Alpha
    OverrideWorldGravity:Vector=Get Gravity Override()
    ExternalForce
    ComponentLinearAccScale:Vector="0.3,0.35,0.35"
    ComponentLinearVelScale:Vector="0.4,0.4,0.4"
   ) -> (Pose) {
    bClampLinearTranslationLimitToRefPose=true
    bOverrideWorldGravity=true
    overridePhysicsAsset=&/F_MED_ChessBoard_FaceAcc_FrontEnd_Physic
    simulationSpace=BaseBoneSpace
    alphaBoolBlend {blendOption=Linear}
    baseBoneRef {boneName=spine_05}
    componentAppliedLinearAccClamp=<5000 5000 5000>
    componentLinearAccScale=<0.3 0.35 0.35>
    componentLinearVelScale=<0.4 0.4 0.4>
   }
   n2 = Component To Local(ComponentPose=@n1)
   n3 = RigidBody(
    ComponentPose=Local To Component(LocalPose=Use cached pose 'Pre_RBAN_2_RBAN'())
    OverrideWorldGravity=Get Gravity Override()
    ComponentLinearAccScale="0.3,0.35,0.35"
    ComponentLinearVelScale="0.4,0.4,0.4"
   ) {
    bClampLinearTranslationLimitToRefPose=true
    bOverrideWorldGravity=true
    overridePhysicsAsset=&/F_MED_ChessBoard_FaceAcc_Front_Physic
    simulationSpace=BaseBoneSpace
    alphaBoolBlend {blendOption=Linear}
    baseBoneRef {boneName=spine_05}
    componentAppliedLinearAccClamp=<5000 5000 5000>
    componentLinearAccScale=<0.3 0.35 0.35>
    componentLinearVelScale=<0.4 0.4 0.4>
   }
   n4 = Component To Local(ComponentPose=@n3)
   n5 = Get bIsInFrontEnd()
   n6 = "Constraint - dyn_fr_lowerhair_l_01 (PP)"!(
    ComponentPose=Local To Component(LocalPose=Blend Poses by bool(BlendPose_0=@n2, BlendPose_1=@n4, bActiveValue=@n5) {blendType=Linear})
    Alpha=0.3
    ConstraintWeights_0=0.5
    ConstraintWeights_1:float=0.5
   ) -> (Pose) {
    alphaBoolBlend {blendOption=Linear}
    boneToModify {boneName=dyn_fr_lowerhair_l_01}
    constraintSetup [{targetBone {boneName=spine_05}}, {targetBone {boneName=clavicle_l}}]
    constraintWeights [1,1]
   }
   n7 = "Constraint - dyn_fr_lowerhair_l_02 (PP)"!(ComponentPose=@n6, Alpha=0.3, ConstraintWeights_0=0.5, ConstraintWeights_1:float=0.5) -> (Pose) {
    alphaBoolBlend {blendOption=Linear}
    boneToModify {boneName=dyn_fr_lowerhair_l_02}
    constraintSetup [{targetBone {boneName=spine_05}}, {targetBone {boneName=clavicle_l}}]
    constraintWeights [1,1]
   }
   n8 = "Constraint - dyn_fr_lowerhair_r_01 (PP)"!(ComponentPose=@n7, Alpha=0.3, ConstraintWeights_0=0.5, ConstraintWeights_1:float=0.5) -> (Pose) {
    alphaBoolBlend {blendOption=Linear}
    boneToModify {boneName=dyn_fr_lowerhair_r_01}
    constraintSetup [{targetBone {boneName=spine_05}}, {targetBone {boneName=clavicle_r}}]
    constraintWeights [1,1]
   }
   n9 = "Constraint - dyn_fr_lowerhair_r_02 (PP)"!(ComponentPose=@n8, Alpha=0.3, ConstraintWeights_0=0.5, ConstraintWeights_1:float=0.5) -> (Pose) {
    alphaBoolBlend {blendOption=Linear}
    boneToModify {boneName=dyn_fr_lowerhair_r_02}
    constraintSetup [{targetBone {boneName=spine_05}}, {targetBone {boneName=clavicle_r}}]
    constraintWeights [1,1]
   }
   n10 = Get R Arm Up() -> (R Arm Up:float)
   n11 = "Transform (Modify) Bone - Bone: dyn_fr_lowerhair_r_01"!(ComponentPose=@n9, Alpha=@n10, Translation="1.045,-0.030,-0.336", Rotation="-30.000,-0.000,0.000") -> (Pose) {
    rotationMode=BMM_Additive
    rotationSpace=BCS_BoneSpace
    translationMode=BMM_Additive
    alphaBoolBlend {blendOption=Linear}
    alphaScaleBiasClamp {bInterpResult=true}
    boneToModify {boneName=dyn_fr_lowerhair_r_01}
    rotation=<rot -30 0 0>
    translation=<1.045 -0.03 -0.336>
   }
   n12 = "Transform (Modify) Bone - Bone: base_fr_hair_r"!(ComponentPose=@n11, Alpha=@n10, Translation="0,0,0", Rotation="-9.847,10.151,-1.754") -> (Pose) {
    rotationMode=BMM_Additive
    rotationSpace=BCS_BoneSpace
    translationMode=BMM_Additive
    alphaBoolBlend {blendOption=Linear}
    alphaScaleBiasClamp {bInterpResult=true}
    boneToModify {boneName=base_fr_hair_r}
    rotation=<rot -9.847 10.151 -1.754>
   }
   n13 = "Transform (Modify) Bone - Bone: dyn_fr_lowerhair_r_02"!(ComponentPose=@n12, Alpha=@n10, Translation="0.026,3.635,1.336", Rotation="0,0,0") -> (Pose) {
    rotationMode=BMM_Additive
    rotationSpace=BCS_BoneSpace
    translationMode=BMM_Additive
    alphaBoolBlend {blendOption=Linear}
    alphaScaleBiasClamp {bInterpResult=true}
    boneToModify {boneName=dyn_fr_lowerhair_r_02}
    translation=<0.026 3.635 1.336>
   }
   n14 = "Transform (Modify) Bone - Bone: dyn_fr_lowerhair_l_01"!(ComponentPose=@n13, Alpha=Get L Arm Up() -> (L Arm Up:float), Translation="0,0,0", Rotation="-0.000,0.000,-40.000") -> (Pose) {
    rotationMode=BMM_Additive
    rotationSpace=BCS_BoneSpace
    translationMode=BMM_Additive
    alphaBoolBlend {blendOption=Linear}
    alphaScaleBiasClamp {bInterpResult=true}
    boneToModify {boneName=dyn_fr_lowerhair_l_01}
    rotation=<rot 0 0 -40>
   }
   n15 = "Transform (Modify) Bone - Bone: dyn_backhair_l_02"(ComponentPose=@n14, Alpha=1, Translation="0,0,0", Rotation="0,0,0", Scale="1,1,1.25") {
    rotationMode=BMM_Additive
    rotationSpace=BCS_BoneSpace
    scaleMode=BMM_Replace
    alphaBoolBlend {blendOption=Linear}
    boneToModify {boneName=dyn_backhair_l_02}
    scale {z=1.25}
   }
   n16 = "Transform (Modify) Bone - Bone: dyn_backhair_m_02"(ComponentPose=@n15, Alpha=1, Translation="0,0,0", Rotation="0,0,0", Scale="1,1,1.25") {
    rotationMode=BMM_Additive
    rotationSpace=BCS_BoneSpace
    scaleMode=BMM_Replace
    alphaBoolBlend {blendOption=Linear}
    boneToModify {boneName=dyn_backhair_m_02}
    scale {z=1.25}
   }
   n17 = "Transform (Modify) Bone - Bone: dyn_backhair_r_02"(ComponentPose=@n16, Alpha=1, Translation="0,0,0", Rotation="0,0,0", Scale="1,1,1.25") {
    rotationMode=BMM_Additive
    rotationSpace=BCS_BoneSpace
    scaleMode=BMM_Replace
    alphaBoolBlend {blendOption=Linear}
    boneToModify {boneName=dyn_backhair_r_02}
    scale {z=1.25}
   }
   n18 = Anim Dynamics - Bone: dyn_frBang_hair_A_01(ComponentPose=@n17, Alpha=1, ExternalForce="0,0,0") {
    angularDampingOverride=1
    angularSpringConstant=5500
    bAngularSpring=true
    bGravityOverrideInSimSpace=true
    bOverrideAngularDamping=true
    bUseGravityOverride=true
    gravityScale=0.35
    simulationSpace=BoneRelative
    alphaBoolBlend {blendOption=Linear}
    boundBone {boneName=dyn_frBang_hair_A_01}
    componentLinearAccScale=<1 1 1>
    componentLinearVelScale=<1 1 1>
    gravityOverride {z=-1200}
    physicsBodyDefinitions [{
      boundBone {boneName=dyn_frBang_hair_A_01}
      boxExtents=<16 16 16>
      constraintSetup {angularLimitsMax=<5 0 5>, angularLimitsMin=<-5 -65 -5>, angularTarget=<1 0 0>}
      localJointOffset=<1 0 0>
     }]
    relativeSpaceBone {boneName=root}
   }
   n19 = Anim Dynamics - Bone: dyn_frBang_hair_A_02(ComponentPose=@n18, Alpha=1, ExternalForce="0,0,0") {
    angularDampingOverride=1
    angularSpringConstant=5500
    bAngularSpring=true
    bGravityOverrideInSimSpace=true
    bOverrideAngularDamping=true
    bUseGravityOverride=true
    gravityScale=0.35
    simulationSpace=BoneRelative
    alphaBoolBlend {blendOption=Linear}
    boundBone {boneName=dyn_frBang_hair_A_02}
    componentLinearAccScale=<1 1 1>
    componentLinearVelScale=<1 1 1>
    gravityOverride {z=-1200}
    physicsBodyDefinitions [{
      boundBone {boneName=dyn_frBang_hair_A_02}
      boxExtents=<16 16 16>
      constraintSetup {angularLimitsMax=<5 0 5>, angularLimitsMin=<-5 -65 -5>, angularTarget=<1 0 0>}
      localJointOffset=<1 0 0>
     }]
    relativeSpaceBone {boneName=root}
   }
   n20 = Anim Dynamics - Bone: dyn_frBang_hair_E_01(ComponentPose=@n19, Alpha=1, ExternalForce="0,0,0") {
    angularDampingOverride=1
    angularSpringConstant=5500
    bAngularSpring=true
    bGravityOverrideInSimSpace=true
    bOverrideAngularDamping=true
    bUseGravityOverride=true
    gravityScale=0.35
    simulationSpace=BoneRelative
    alphaBoolBlend {blendOption=Linear}
    boundBone {boneName=dyn_frBang_hair_E_01}
    componentLinearAccScale=<1 1 1>
    componentLinearVelScale=<1 1 1>
    gravityOverride {z=-1200}
    physicsBodyDefinitions [{
      boundBone {boneName=dyn_frBang_hair_E_01}
      boxExtents=<16 16 16>
      constraintSetup {angularLimitsMax=<5 0 5>, angularLimitsMin=<-5 -65 -5>, angularTarget=<1 0 0>}
      localJointOffset=<1 0 0>
     }]
    relativeSpaceBone {boneName=root}
   }
   n21 = Anim Dynamics - Bone: dyn_frBang_hair_D_01(ComponentPose=@n20, Alpha=1, ExternalForce="0,0,0") {
    angularDampingOverride=1
    angularSpringConstant=5500
    bAngularSpring=true
    bGravityOverrideInSimSpace=true
    bOverrideAngularDamping=true
    bUseGravityOverride=true
    gravityScale=0.35
    simulationSpace=BoneRelative
    alphaBoolBlend {blendOption=Linear}
    boundBone {boneName=dyn_frBang_hair_D_01}
    componentLinearAccScale=<1 1 1>
    componentLinearVelScale=<1 1 1>
    gravityOverride {z=-1200}
    physicsBodyDefinitions [{
      boundBone {boneName=dyn_frBang_hair_D_01}
      boxExtents=<16 16 16>
      constraintSetup {angularLimitsMax=<5 0 5>, angularLimitsMin=<-5 -65 -5>, angularTarget=<1 0 0>}
      localJointOffset=<1 0 0>
     }]
    relativeSpaceBone {boneName=root}
   }
   n22 = Anim Dynamics - Bone: dyn_frBang_hair_B_01(ComponentPose=@n21, Alpha=1, ExternalForce="0,0,0") {
    angularDampingOverride=1
    angularSpringConstant=5500
    bAngularSpring=true
    bGravityOverrideInSimSpace=true
    bOverrideAngularDamping=true
    bUseGravityOverride=true
    gravityScale=0.35
    simulationSpace=BoneRelative
    alphaBoolBlend {blendOption=Linear}
    boundBone {boneName=dyn_frBang_hair_B_01}
    componentLinearAccScale=<1 1 1>
    componentLinearVelScale=<1 1 1>
    gravityOverride {z=-1200}
    physicsBodyDefinitions [{
      boundBone {boneName=dyn_frBang_hair_B_01}
      boxExtents=<16 16 16>
      constraintSetup {angularLimitsMax=<5 0 5>, angularLimitsMin=<-5 -65 -5>, angularTarget=<1 0 0>}
      localJointOffset=<1 0 0>
     }]
    relativeSpaceBone {boneName=root}
   }
   n23 = Anim Dynamics - Bone: dyn_frBang_hair_C_01(ComponentPose=@n22, Alpha=1, ExternalForce="0,0,0") {
    angularDampingOverride=1
    angularSpringConstant=5500
    bAngularSpring=true
    bGravityOverrideInSimSpace=true
    bOverrideAngularDamping=true
    bUseGravityOverride=true
    gravityScale=0.35
    simulationSpace=BoneRelative
    alphaBoolBlend {blendOption=Linear}
    boundBone {boneName=dyn_frBang_hair_C_01}
    componentLinearAccScale=<1 1 1>
    componentLinearVelScale=<1 1 1>
    gravityOverride {z=-1200}
    physicsBodyDefinitions [{
      boundBone {boneName=dyn_frBang_hair_C_01}
      boxExtents=<16 16 16>
      constraintSetup {angularLimitsMax=<0 0 0>, angularLimitsMin=<-2 -40 -2>, angularTarget=<1 0 0>}
      localJointOffset=<1 0 0>
     }]
    relativeSpaceBone {boneName=root}
   }
   n24 = Anim Dynamics - Bone: dyn_frBang_hair_D_02(ComponentPose=@n23, Alpha=1, ExternalForce="0,0,0") {
    angularDampingOverride=1
    angularSpringConstant=5500
    bAngularSpring=true
    bGravityOverrideInSimSpace=true
    bOverrideAngularDamping=true
    bUseGravityOverride=true
    gravityScale=0.35
    simulationSpace=BoneRelative
    alphaBoolBlend {blendOption=Linear}
    boundBone {boneName=dyn_frBang_hair_D_02}
    componentLinearAccScale=<1 1 1>
    componentLinearVelScale=<1 1 1>
    gravityOverride {z=-1200}
    physicsBodyDefinitions [{
      boundBone {boneName=dyn_frBang_hair_D_02}
      boxExtents=<16 16 16>
      constraintSetup {angularLimitsMax=<5 0 5>, angularLimitsMin=<-5 -65 -5>, angularTarget=<1 0 0>}
      localJointOffset=<1 0 0>
     }]
    relativeSpaceBone {boneName=root}
   }
   n25 = Anim Dynamics - Bone: dyn_backhair_l_04(ComponentPose=@n24, Alpha=1, ExternalForce="0,0,0") {
    angularDampingOverride=1
    angularSpringConstant=5500
    bAngularSpring=true
    bGravityOverrideInSimSpace=true
    bOverrideAngularDamping=true
    bUseGravityOverride=true
    gravityScale=0.35
    simulationSpace=BoneRelative
    alphaBoolBlend {blendOption=Linear}
    boundBone {boneName=dyn_backhair_l_04}
    componentLinearAccScale=<1 1 1>
    componentLinearVelScale=<1 1 1>
    gravityOverride {z=-1200}
    physicsBodyDefinitions [{
      boundBone {boneName=dyn_backhair_l_04}
      boxExtents=<13 13 13>
      constraintSetup {angularLimitsMax=<0 0 0>, angularLimitsMin=<0 -65 0>, angularTarget=<1 0 0>}
      localJointOffset=<1 0 0>
     }]
    relativeSpaceBone {boneName=root}
   }
   n26 = Anim Dynamics - Bone: dyn_backhair_m_05(ComponentPose=@n25, Alpha=1, ExternalForce="0,0,0") {
    angularDampingOverride=1
    angularSpringConstant=5500
    bAngularSpring=true
    bGravityOverrideInSimSpace=true
    bOverrideAngularDamping=true
    bUseGravityOverride=true
    gravityScale=0.35
    simulationSpace=BoneRelative
    alphaBoolBlend {blendOption=Linear}
    boundBone {boneName=dyn_backhair_m_05}
    componentLinearAccScale=<1 1 1>
    componentLinearVelScale=<1 1 1>
    gravityOverride {z=-1200}
    physicsBodyDefinitions [{
      boundBone {boneName=dyn_backhair_m_05}
      boxExtents=<13 13 13>
      constraintSetup {angularLimitsMax=<0 0 0>, angularLimitsMin=<0 -65 0>, angularTarget=<1 0 0>}
      localJointOffset=<1 0 0>
     }]
    relativeSpaceBone {boneName=root}
   }
   n27 = Anim Dynamics - Bone: dyn_backhair_r_04(ComponentPose=@n26, Alpha=1, ExternalForce="0,0,0") {
    angularDampingOverride=1
    angularSpringConstant=5500
    bAngularSpring=true
    bGravityOverrideInSimSpace=true
    bOverrideAngularDamping=true
    bUseGravityOverride=true
    gravityScale=0.35
    simulationSpace=BoneRelative
    alphaBoolBlend {blendOption=Linear}
    boundBone {boneName=dyn_backhair_r_04}
    componentLinearAccScale=<1 1 1>
    componentLinearVelScale=<1 1 1>
    gravityOverride {z=-1200}
    physicsBodyDefinitions [{
      boundBone {boneName=dyn_backhair_r_04}
      boxExtents=<13 13 13>
      constraintSetup {angularLimitsMax=<0 0 0>, angularLimitsMin=<0 -65 0>, angularTarget=<1 0 0>}
      localJointOffset=<1 0 0>
     }]
    relativeSpaceBone {boneName=root}
   }
   n28 = "Transform (Modify) Bone - Bone: base_backhair_l"(ComponentPose=@n27, Alpha=1, Translation="0.000,-0.000,1.238", Rotation="0,-10,20", Scale="1,1,1") {
    rotationMode=BMM_Additive
    scaleMode=BMM_Replace
    translationMode=BMM_Additive
    alphaBoolBlend {blendOption=Linear}
    alphaScaleBiasClamp {bInterpResult=true, interpSpeedDecreasing=15, interpSpeedIncreasing=15}
    boneToModify {boneName=base_backhair_l}
    rotation {roll=20, yaw=-10}
    translation {z=1.238}
   }
   n29 = "Transform (Modify) Bone - Bone: base_backhair_m"(ComponentPose=@n28, Alpha=1, Translation="0,0,0", Rotation="0,0,0", Scale="1,1,1") {
    rotationMode=BMM_Additive
    scaleMode=BMM_Replace
    translationMode=BMM_Additive
    alphaBoolBlend {blendOption=Linear}
    alphaScaleBiasClamp {bInterpResult=true, interpSpeedDecreasing=15, interpSpeedIncreasing=15}
    boneToModify {boneName=base_backhair_m}
   }
   n30 = "Transform (Modify) Bone - Bone: base_backhair_r"(ComponentPose=@n29, Alpha=1, Translation="0,0,0", Rotation="-11.578,10.053,-1.887", Scale="1,1,1") {
    rotationMode=BMM_Additive
    scaleMode=BMM_Replace
    translationMode=BMM_Additive
    alphaBoolBlend {blendOption=Linear}
    alphaScaleBiasClamp {bInterpResult=true, interpSpeedDecreasing=15, interpSpeedIncreasing=15}
    boneToModify {boneName=base_backhair_r}
    rotation=<rot -11.578 10.053 -1.887>
   }
   n31 = "Constraint - dyn_backhair_l_02 (PPPP)"!(
    ComponentPose=@n30
    Alpha=1
    ConstraintWeights_0=0.2
    ConstraintWeights_1:float=0.3
    ConstraintWeights_2:float=0.3
    ConstraintWeights_3:float=0.2
   ) -> (Pose) {
    alphaBoolBlend {blendOption=Linear}
    boneToModify {boneName=dyn_backhair_l_02}
    constraintSetup [
     {targetBone {boneName=dyn_backhair_l_01}}
     {targetBone {boneName=neck_02}}
     {targetBone {boneName=spine_05}}
     {targetBone {boneName=neck_01}}
    ]
    constraintWeights [1,1,1,1]
   }
   n32 = "Constraint - dyn_backhair_l_03 (PP)"!(ComponentPose=@n31, Alpha=1, ConstraintWeights_0=0.7, ConstraintWeights_1:float=0.3) -> (Pose) {
    alphaBoolBlend {blendOption=Linear}
    boneToModify {boneName=dyn_backhair_l_03}
    constraintSetup [{targetBone {boneName=spine_05}}, {targetBone {boneName=neck_01}}]
    constraintWeights [1,1]
   }
   n33 = "Constraint - dyn_backhair_l_04 (P)"(ComponentPose=@n32, Alpha=1, ConstraintWeights_0=1) {
    alphaBoolBlend {blendOption=Linear}
    boneToModify {boneName=dyn_backhair_l_04}
    constraintSetup [{targetBone {boneName=spine_05}}]
    constraintWeights [1]
   }
   n34 = "Constraint - dyn_backhair_l_05 (PPP)"!(ComponentPose=@n33, Alpha=1, ConstraintWeights_0=0.2, ConstraintWeights_1:float=0.8, ConstraintWeights_2:float=0) -> (Pose) {
    alphaBoolBlend {blendOption=Linear}
    boneToModify {boneName=dyn_backhair_l_05}
    constraintSetup [{targetBone {boneName=spine_05}}, {targetBone {boneName=clavicle_l}}, {targetBone {boneName=upperarm_l}}]
    constraintWeights [1,1,1]
   }
   n35 = "Constraint - dyn_backhair_m_02 (PPPP)"!(
    ComponentPose=@n34
    Alpha=1
    ConstraintWeights_0=0.2
    ConstraintWeights_1:float=0.3
    ConstraintWeights_2:float=0.3
    ConstraintWeights_3:float=0.2
   ) -> (Pose) {
    alphaBoolBlend {blendOption=Linear}
    boneToModify {boneName=dyn_backhair_m_02}
    constraintSetup [
     {targetBone {boneName=dyn_backhair_m_01}}
     {targetBone {boneName=neck_02}}
     {targetBone {boneName=spine_05}}
     {targetBone {boneName=neck_01}}
    ]
    constraintWeights [1,1,1,1]
   }
   n36 = "Constraint - dyn_backhair_m_03 (PP)"!(ComponentPose=@n35, Alpha=1, ConstraintWeights_0=0.7, ConstraintWeights_1:float=0.3) -> (Pose) {
    alphaBoolBlend {blendOption=Linear}
    boneToModify {boneName=dyn_backhair_m_03}
    constraintSetup [{targetBone {boneName=spine_05}}, {targetBone {boneName=neck_01}}]
    constraintWeights [1,1]
   }
   n37 = "Constraint - dyn_backhair_m_04 (P)"(ComponentPose=@n36, Alpha=1, ConstraintWeights_0=1) {
    alphaBoolBlend {blendOption=Linear}
    boneToModify {boneName=dyn_backhair_m_04}
    constraintSetup [{targetBone {boneName=spine_05}}]
    constraintWeights [1]
   }
   n38 = "Constraint - dyn_backhair_m_05 (P)"(ComponentPose=@n37, Alpha=1, ConstraintWeights_0=1) {
    alphaBoolBlend {blendOption=Linear}
    boneToModify {boneName=dyn_backhair_m_05}
    constraintSetup [{targetBone {boneName=spine_05}}]
    constraintWeights [1]
   }
   n39 = "Constraint - dyn_backhair_r_02 (PPPP)"!(
    ComponentPose=@n38
    Alpha=1
    ConstraintWeights_0=0.2
    ConstraintWeights_1:float=0.3
    ConstraintWeights_2:float=0.3
    ConstraintWeights_3:float=0.2
   ) -> (Pose) {
    alphaBoolBlend {blendOption=Linear}
    boneToModify {boneName=dyn_backhair_r_02}
    constraintSetup [
     {targetBone {boneName=dyn_backhair_r_01}}
     {targetBone {boneName=neck_02}}
     {targetBone {boneName=spine_05}}
     {targetBone {boneName=neck_01}}
    ]
    constraintWeights [1,1,1,1]
   }
   n40 = "Constraint - dyn_backhair_r_03 (PP)"!(ComponentPose=@n39, Alpha=1, ConstraintWeights_0=0.7, ConstraintWeights_1:float=0.3) -> (Pose) {
    alphaBoolBlend {blendOption=Linear}
    boneToModify {boneName=dyn_backhair_r_03}
    constraintSetup [{targetBone {boneName=spine_05}}, {targetBone {boneName=neck_01}}]
    constraintWeights [1,1]
   }
   n41 = "Constraint - dyn_backhair_r_04 (P)"(ComponentPose=@n40, Alpha=1, ConstraintWeights_0=1) {
    alphaBoolBlend {blendOption=Linear}
    boneToModify {boneName=dyn_backhair_r_04}
    constraintSetup [{targetBone {boneName=spine_05}}]
    constraintWeights [1]
   }
   n42 = "Constraint - dyn_backhair_r_05 (PPP)"!(ComponentPose=@n41, Alpha=1, ConstraintWeights_0=0.2, ConstraintWeights_1:float=0.8, ConstraintWeights_2:float=0) -> (Pose) {
    alphaBoolBlend {blendOption=Linear}
    boneToModify {boneName=dyn_backhair_r_05}
    constraintSetup [{targetBone {boneName=spine_05}}, {targetBone {boneName=clavicle_r}}, {targetBone {boneName=upperarm_r}}]
    constraintWeights [1,1,1]
   }
   n43 = "Transform (Modify) Bone - Bone: dyn_backhair_l_02"(ComponentPose=@n42, Alpha=1, Translation="0.000,-1.105,-0.771", Rotation="0.000,0.000,25.000", Scale="1,1,1") {
    rotationMode=BMM_Additive
    scaleMode=BMM_Replace
    translationMode=BMM_Additive
    alphaBoolBlend {blendOption=Linear}
    alphaScaleBiasClamp {bInterpResult=true, interpSpeedDecreasing=5, interpSpeedIncreasing=5}
    boneToModify {boneName=dyn_backhair_l_02}
    rotation {roll=25}
    translation {y=-1.105, z=-0.771}
   }
   n44 = "Transform (Modify) Bone - Bone: dyn_backhair_m_02"(ComponentPose=@n43, Alpha=1, Translation="-0.000,-3.415,-1.080", Rotation="0,0,5", Scale="1,1,1") {
    rotationMode=BMM_Additive
    scaleMode=BMM_Replace
    translationMode=BMM_Additive
    alphaBoolBlend {blendOption=Linear}
    alphaScaleBiasClamp {bInterpResult=true, interpSpeedDecreasing=5, interpSpeedIncreasing=5}
    boneToModify {boneName=dyn_backhair_m_02}
    rotation {roll=5}
    translation=<0 -3.415 -1.0797>
   }
   n45 = "Transform (Modify) Bone - Bone: dyn_backhair_r_02"(ComponentPose=@n44, Alpha=1, Translation="-0.000,-1.284,-0.833", Rotation="0.000,0.000,15.000", Scale="1,1,1") {
    rotationMode=BMM_Additive
    scaleMode=BMM_Replace
    translationMode=BMM_Additive
    alphaBoolBlend {blendOption=Linear}
    alphaScaleBiasClamp {bInterpResult=true, interpSpeedDecreasing=5, interpSpeedIncreasing=5}
    boneToModify {boneName=dyn_backhair_r_02}
    rotation {roll=15}
    translation {y=-1.284, z=-0.833}
   }
   n46 = "Transform (Modify) Bone - Bone: dyn_backhair_r_03"(ComponentPose=@n45, Alpha=1, Translation="-0.453,0.000,2.648", Rotation="-10.000,0.000,-0.000", Scale="1,1,1") {
    rotationMode=BMM_Additive
    scaleMode=BMM_Replace
    translationMode=BMM_Additive
    alphaBoolBlend {blendOption=Linear}
    alphaScaleBiasClamp {bInterpResult=true, interpSpeedDecreasing=5, interpSpeedIncreasing=5}
    boneToModify {boneName=dyn_backhair_r_03}
    rotation {pitch=-10}
    translation {x=-0.453, z=2.6483}
   }
   n47 = "Transform (Modify) Bone - Bone: dyn_backhair_l_03"(ComponentPose=@n46, Alpha=1, Translation="-1.009,-2.967,2.157", Rotation="0,0,0", Scale="1,1,1") {
    rotationMode=BMM_Additive
    scaleMode=BMM_Replace
    translationMode=BMM_Additive
    alphaBoolBlend {blendOption=Linear}
    alphaScaleBiasClamp {bInterpResult=true, interpSpeedDecreasing=5, interpSpeedIncreasing=5}
    boneToModify {boneName=dyn_backhair_l_03}
    translation=<-1.009 -2.967 2.1565>
   }
   n48 = "Transform (Modify) Bone - Bone: dyn_backhair_r_04"(ComponentPose=@n47, Alpha=1, Translation="0.821,-0.000,0.000", Rotation="0,0,0", Scale="1,1,1") {
    rotationMode=BMM_Additive
    scaleMode=BMM_Replace
    translationMode=BMM_Additive
    alphaBoolBlend {blendOption=Linear}
    alphaScaleBiasClamp {bInterpResult=true, interpSpeedDecreasing=5, interpSpeedIncreasing=5}
    boneToModify {boneName=dyn_backhair_r_04}
    translation {x=0.821}
   }
   n49 = "Transform (Modify) Bone - Bone: dyn_backhair_r_05"(ComponentPose=@n48, Alpha=1, Translation="-0.000,0.000,1.625", Rotation="-20.000,0.000,-0.000", Scale="1,1,1") {
    rotationMode=BMM_Additive
    scaleMode=BMM_Replace
    translationMode=BMM_Additive
    alphaBoolBlend {blendOption=Linear}
    alphaScaleBiasClamp {bInterpResult=true, interpSpeedDecreasing=5, interpSpeedIncreasing=5}
    boneToModify {boneName=dyn_backhair_r_05}
    rotation {pitch=-20}
    translation {z=1.625}
   }
   Output Pose(Result=Component To Local(ComponentPose=@n49)) {layerGroup=None}
   n50 = Get Body Mesh() -> (Body Mesh:Skeletal Mesh Component Object Reference)
   n51 = Control Rig!(
    Source=Layered blend per bone(
     BasePose=Copy Pose From Mesh() {bUseAttachedParent=true}
     BlendPoses_0=Copy Pose From Mesh(SourceMeshComponent=@n50) {bUseAttachedParent=true}
    ) {layerSetup [{branchFilters [{boneName=clavicle_l}, {boneName=clavicle_r}]}]}
    Alpha
    is crouching:bool=Get bIsCrouching()
   ) -> (Pose) {
    alphaBoolBlend {blendOption=Linear}
    controlRigAssetReference {blueprintRigClass=&/F_MED_ChessBoard_FaceAcc_CtrlRig.F_MED_ChessBoard_FaceAcc_CtrlRig_C}
   }
   Save cached pose 'Start pose'(Pose=@n51) {cacheName=Start pose}
   n52 = Get Head Up() -> (Head Up:float)
   n53 = "Transform (Modify) Bone - Bone: base_backhair_l"(
    ComponentPose=Local To Component(LocalPose=Use cached pose 'Start pose'())
    Alpha=@n52
    Translation="0,0,0"
    Rotation="40,0.003,0.002"
    Scale="1.000,1.000,1.500"
   ) {
    rotationMode=BMM_Additive
    rotationSpace=BCS_BoneSpace
    scaleMode=BMM_Replace
    alphaBoolBlend {blendOption=Linear}
    boneToModify {boneName=base_backhair_l}
    rotation=<rot 40 0.003 0.002>
    scale {z=1.5}
   }
   n54 = "Transform (Modify) Bone - Bone: base_backhair_m"(ComponentPose=@n53, Alpha=@n52, Translation="0,0,0", Rotation="60,0,0", Scale="1,1,1") {
    rotationMode=BMM_Additive
    rotationSpace=BCS_BoneSpace
    scaleMode=BMM_Replace
    alphaBoolBlend {blendOption=Linear}
    boneToModify {boneName=base_backhair_m}
    rotation {pitch=60}
   }
   n55 = "Transform (Modify) Bone - Bone: base_backhair_r"(ComponentPose=@n54, Alpha=@n52, Translation="0,0,0", Rotation="60,0,0", Scale="1,1,1") {
    rotationMode=BMM_Additive
    rotationSpace=BCS_BoneSpace
    scaleMode=BMM_Replace
    alphaBoolBlend {blendOption=Linear}
    boneToModify {boneName=base_backhair_r}
    rotation {pitch=60}
   }
   Save cached pose 'Pre_RBAN_2_RBAN'(Pose=Component To Local(ComponentPose=@n55)) {cacheName=Pre_RBAN_2_RBAN}
   Get Head Up()
  ]}
 EventGraph {nodes [
   Event BlueprintUpdateAnimation()
   > n1 = Sequence!(execute) -> (then_0, then_1, then_2:Exec, then_3:Exec)
   > n2 = Gravity Override State Function DEPRECATED(CharacterPartAnimInstance=Self-Reference(), GameTimeInSeconds=GetGameTimeInSeconds(), GravityOverrideParams=Get Gravity Override Params())
   > Set Gravity Override(Gravity Override=@n2)
   Set Head Up(
    execute=@n1.then_1
    Head Up:float=GetAlphaFromReferenceTransformDelta(
     SkeletalMeshComponent=Get Skel Mesh() -> (Skel Mesh:Skinned Mesh Component Object Reference)
     BoneName=spine_04
     BaseName=head
     InputRange=(X=15,Y=60)
     TwistAxis="0,0,1"
    )
   ) -> (Output_Get:float)
   > Set Head Down(Head Down:double=GetAlphaFromReferenceTransformDelta(
     SkeletalMeshComponent=Get Skel Mesh()
     BoneName=spine_05
     BaseName="head "
     InputRange=(X=0,Y=-90)
     OutputRange=(X=0,Y=-1)
     TwistAxis="0,0,1"
    )) -> (Output_Get:double)
   n3 = BreakTransform(InTransform=GetDeltaTransformFromRefPose(self=GetOwnerSkeletalMeshComponent(), BoneName=upperarm_r, BaseName=clavicle_r))
   > Set R Arm Up(R Arm Up:float=MapRangeClamped(Value=BreakRotator(InRot=@n3.Rotation).Yaw, InRangeA=55, InRangeB=90)) -> (Output_Get:float)
   n4 = BreakTransform(InTransform=GetDeltaTransformFromRefPose(self=GetOwnerSkeletalMeshComponent(), BoneName=upperarm_l, BaseName=clavicle_l))
   > Set L Arm Up(L Arm Up:float=MapRangeClamped(Value=BreakRotator(InRot=@n4.Rotation).Yaw, InRangeA=20, InRangeB=50)) -> (Output_Get:float)
   n5 = Branch(execute=@n1.then_2, Condition=Get Is Backbling Equipped() -> (Is Backbling Equipped:bool))
   > Set Front hair Comp Acc(Front hair Comp Acc:Vector="0, 0, 0") -> (Output_Get:Vector)
   > Set Front hair Comp Vel(Front hair Comp Vel:Vector="0, 0, 0") -> (Output_Get:Vector)
   > Set backbling constraint alpha(backbling constraint alpha:double=0.6) -> (Output_Get:double)
   n6 = "To Float (Boolean)"(InBool=Get bIsSkydiveFloating())
   > Set Is Backpack Skydive Floating(Is Backpack Skydive Floating:double=@n6) -> (Output_Get:double)
   > Set Is Skydive Floating NoPack(Is Skydive Floating NoPack:double=0) -> (Output_Get:double)
   Set Front hair Comp Acc(execute=@n5.else, Front hair Comp Acc="0.3,0.35,0.35")
   > Set Front hair Comp Vel(Front hair Comp Vel="0.4,0.4,0.4")
   > Set backbling constraint alpha(backbling constraint alpha=0)
   > Set Is Backpack Skydive Floating(Is Backpack Skydive Floating=0)
   > Set Is Skydive Floating NoPack(Is Skydive Floating NoPack=@n6)
   Event BlueprintInitializeAnimation()
   > Set Skel Mesh(Skel Mesh:Skinned Mesh Component Object Reference=GetOwnerSkeletalMeshComponent()) -> (Output_Get:Skinned Mesh Component Object Reference)
   Event OnCharacterCustomizationLoaded()
   > n7 = Cast To PlayerPawn_Athena(Object=TryGetPawnOwner())
   > Set Body Mesh(Body Mesh:Skeletal Mesh Component Object Reference=GetSkeletalMeshForPartType(self=@n7)) -> (Output_Get:Skeletal Mesh Component Object Reference)
   n8 = CheckForBackbling()
   > n9 = Set Is Wearing Backpack(Is Wearing Backpack:bool=AND Boolean!(A=@n8, B=NOT Boolean(A=@n8.bIsBackblingACape), C:bool=NOT Boolean(A=@n8.bIsBackblingLong)) -> (ReturnValue)) -> (Output_Get:bool)
   > n10 = Set Is Wearing Cape(Is Wearing Cape=OR Boolean(A=@n8.bIsBackblingACape, B=@n8.bIsBackblingLong))
   > Set Is Backbling Equipped(Is Backbling Equipped:bool=OR Boolean(A=@n10, B=@n9)) -> (Output_Get:bool)
   > Set is SD with BP Equipped(is SD with BP Equipped:bool=AND Boolean!(A=Get Is Backbling Equipped(), B=Get bIsSkydiveFloating()) -> (ReturnValue)) -> (Output_Get:bool)
   Set Rot Head L(Rot Head L:float=GetAlphaFromReferenceTransformDelta(
     SkeletalMeshComponent=Get Skel Mesh()
     BoneName=neck_01
     BaseName="head "
     DeltaComponentType=EulerX
     InputRange=(X=-10,Y=-42)
     TwistAxis="1,0,0"
    )) -> (Output_Get:float)
   > Set Rot Head R(Rot Head R:float=GetAlphaFromReferenceTransformDelta(
     SkeletalMeshComponent=Get Skel Mesh()
     BoneName=neck_01
     BaseName="head "
     DeltaComponentType=EulerX
     InputRange=(X=10,Y=42)
     TwistAxis="1,0,0"
    )) -> (Output_Get:float)
   > Set Rot ClavicleArm_R(Rot ClavicleArm_R:float=GetAlphaFromReferenceTransformDelta(SkeletalMeshComponent=Get Skel Mesh(), BoneName=clavicle_r, BaseName=upperarm_r, InputRange=(X=0,Y=-10), TwistAxis="1,0,0")) -> (Output_Get:float)
   Set Is Aiming(Is Aiming:double="To Float (Boolean)"(InBool=Get bIsTargeting())) -> (Output_Get:double)
   > Set Is parachuting(Is parachuting:double="To Float (Boolean)"(InBool=Get bIsParachuteOpened())) -> (Output_Get:double)
   > Set Is Aiming Head Up(Is Aiming Head Up:double=float * float(A=Get Is Aiming() -> (Is Aiming:double), B=Get Head Up()) {is=title K2Node_PromotableOperator}) -> (Output_Get:double)
   Get Is Backbling Equipped()
   Get bIsParachuteOpened()
   Set Is Backpack Skydive Floating(Is Backpack Skydive Floating=0)
  ]}
]
variables [
 backbling constraint alpha {type=number}
 Body Mesh {type=SkeletalMeshComponent}
 Front hair Comp Acc {type=Vector}
 Front hair Comp Vel {type=Vector}
 Gravity Override {type=Vector}
 Gravity Override Params {
  type=GravityOverrideParamsStruct
  default {
   skydive_FloatingDown {gravityOverride=<-300 -1500 600>}
   skydive_GoingUp {gravityOverride=<-300 -1500 1500>}
   skydive_Parachuting {gravityOverride=<0 -800 -1500>}
  }
 }
 Head Down {type=number}
 Head Up {type=number}
 HeadRot {type=number}
 Is Aiming {type=number}
 Is Aiming Head Up {type=number}
 Is Backbling Equipped {type=boolean}
 Is Backpack Skydive Floating {type=number}
 Is parachuting {type=number}
 is SD with BP Equipped {type=boolean}
 Is Skydive Floating NoPack {type=number}
 Is Wearing Backpack {type=boolean}
 Is Wearing Cape {type=boolean}
 L Arm Up {type=number}
 R Arm Up {type=number}
 Rot ClavicleArm_R {type=number}
 Rot Head L {type=number}
 Rot Head R {type=number}
 Skel Mesh {type=SkinnedMeshComponent, default=null}
]
